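# Looping day ambience Sound Cue for the wet room of the SuperVillan base POI, in the indoor ambient sound class.
# Nested mixers layer a Hollow Base bathroom loop, weighted loudest, over a quieter submix of an industrial room tone and a haunted hotel night large-room loop.
# It has no spatialization or attenuation, and its concurrency override prevents new instances once the limit is reached.
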SoundCue /Game/Sounds/Artemis/POI/SuperVillan/Ambience_SuperVillanBase_WetRoom_Day_Loop_Cue
root=Mixer_1
nodes [
 {id=Mixer_0, children [WavePlayer_3,WavePlayer_0], properties {inputVolume [1,1]}}
 {id=Mixer_1, children [WavePlayer_1,Mixer_0], properties {inputVolume [1.5,0.5]}}
 {
  id=WavePlayer_0
  properties {bLooping=true, soundWaveAssetPtr=/Game/Sounds/Fort_Starlight/S4/SoundWaves/AMBI_Starlight_S4_HauntedHotel_LargeRoom_Int_Night_Loop}
 }
 {
  id=WavePlayer_1
  properties {bLooping=true, soundWaveAssetPtr=/Game/Sounds/Artemis/POI/Hollow_Base/Ambience_Hollow_Base_Bathroom_Int_Day_Loop}
 }
 {id=WavePlayer_3, properties {bLooping=true, soundWaveAssetPtr=/Game/Sounds/Trans_AmbientLoops/RoomTones/IndustrialSmall002_Loop_Stereo}}
]
properties {
 bOverrideConcurrency=true
 maxDistance=2097152
 soundClassObject=&Fort_Ambient_Indoors_Athena
 volumeMultiplier=0.501
 attenuationOverrides {bAttenuate=false, bEnableReverbSend=false, bSpatialize=false}
 concurrencyOverrides {resolutionRule=PreventNew}
}
