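# Night ambience loop cue for the wet room of the SuperVillan base POI, in the Athena indoor ambient sound class.
# A root mixer blends looping bathroom and church interior beds with a nested mixer of industrial room tone and crypt loop, weighted toward bathroom and crypt.
# All layers reuse waves from other sound folders.
# Attenuation, spatialization and reverb send are off, and concurrency prevents new instances once at the limit.
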SoundCue /Game/Sounds/Artemis/POI/SuperVillan/Ambience_SuperVillanBase_WetRoom_Night_Loop_Cue
root=Mixer_0
nodes [
 {id=Mixer_0, children [WavePlayer_1,WavePlayer_2,Mixer_1], properties {inputVolume [1.75,0.5,0.5]}}
 {id=Mixer_1, children [WavePlayer_4,WavePlayer_3], properties {inputVolume [0.5,2]}}
 {
  id=WavePlayer_1
  properties {bLooping=true, soundWaveAssetPtr=/Game/Sounds/Artemis/POI/Hollow_Base/Ambience_Hollow_Base_Bathroom_Int_Day_Loop}
 }
 {
  id=WavePlayer_2
  properties {bLooping=true, soundWaveAssetPtr=/Game/Sounds/Fort_Starlight/S1/SoundWaves/AMBI_Starlight_mToD_Church_Int_Day_Loop}
 }
 {
  id=WavePlayer_3
  properties {bLooping=true, soundWaveAssetPtr=/Game/Sounds/Fort_Starlight/S1/SoundWaves/AMBI_Starlight_mToD_Crypt_Int_Morning_02_Loop}
 }
 {id=WavePlayer_4, properties {bLooping=true, soundWaveAssetPtr=/Game/Sounds/Trans_AmbientLoops/RoomTones/IndustrialSmall002_Loop_Stereo}}
]
properties {
 bOverrideConcurrency=true
 maxDistance=2097152
 soundClassObject=&Fort_Ambient_Indoors_Athena
 volumeMultiplier=0.501
 attenuationOverrides {bAttenuate=false, bEnableReverbSend=false, bSpatialize=false}
 concurrencyOverrides {resolutionRule=PreventNew}
}
